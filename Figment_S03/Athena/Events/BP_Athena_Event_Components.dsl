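; Season-event actor that stages a comet in the sky, growing and moving through phases as the end of season nears.
; DefineEventPhase derives the phase on authority from calendar timing and data-table day thresholds, with a debug override.
; SetupComet applies phase-specific transforms, curve-driven scale and ambient audio, rerunning when CometPosition replicates.
; Later phases enable randomly timed shooting stars and meteors, scheduled on the non-cosmetic branch and spawned off dedicated servers.
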(event Custom|UpdateComet
  (CallFunction|SetupComet))

(event EventBeginPlay
  (Class|FortAsyncActionCurrentPlaylistReady|PlaylistDataReadyAsync
    (:OnReady
      (MCPContext|InitialCalendarSyncCompleted
        (:OnSyncComplete
          (CallFunction|DefineEventPhase)
          (Networking|CanExecuteCosmeticEvents
            (:True
              (CallFunction|SetupComet))
            (:False
              ;[1 1 1] Change comet size/position based on number of days away from End of Season
              (Utilities|SetCometPositionValue)
              (Utilities|FlowControl|Delay 5.0)
              (CallFunction|UpdateComet)
              (Utilities|FlowControl|Sequence
                (:then_0
                  ;[1 1 1] Looping shooting star spawn event with random delay
                  (if (or (Variables|Default|GetShootingStarsActive) (Variables|Debug|GetPreviewShootingStars))
                    (Utilities|FlowControl|Delay (* (Math|Random|RandomFloatinRange 0.9 1.6) (Math|Float|MapRangeClamped (Math|Float|SelectFloat 1.0 (Variables|Default|GetTimeSpanRatio) (Variables|Debug|GetPreviewShootingStars)) 0.0 1.0 35.0 1e+01)))
                    (CallFunction|RandomizeSpawnLocation)
                    (CallFunction|SpawnShootingStars)))
                (:then_1
                  ;[1 1 1] Looping meteor spawn event with random delay
                  (if (or (Variables|Default|GetMeteorsActive) (Variables|Debug|GetPreviewMeteors))
                    (Utilities|FlowControl|Delay (* (Math|Random|RandomFloatinRange 0.7 2.0) (Math|Float|MapRangeClamped (Math|Float|SelectFloat 1.0 (Variables|Default|GetTimeSpanRatio) (Variables|Debug|GetPreviewMeteors)) 0.0 1.0 27.0 8.0)))
                    (CallFunction|RandomizeSpawnLocation)
                    (CallFunction|SpawnMeteor)))))) ; Must be on a dedicated server
        )))))

(event Custom|SpawnShootingStars
  ;[1 1 1] Spawn shooting star particle system
  (if (or (Variables|Default|GetShootingStarsActive) (Variables|Debug|GetPreviewShootingStars))
    (if (not (Networking|IsDedicatedServer))
      (bind _returnvalue (Effects|SpawnEmitteratLocation "/Figment_S03/Athena/Events/P_Comet_ShootingStars.P_Comet_ShootingStars" (Utilities|Array|Get(acopy) (Variables|Default|GetShootingStarSpawns) (Variables|Default|GetSpawnLocation)) (Transformation|GetWorldRotation (Variables|Default|GetP_Athena_Event_Comet)) (Math|Conversions|ToVector(Float) (Math|Float|Lerp 0.7 1.0 (Variables|Default|GetTimeSpanRatio)))))
      (Variables|/Script/CascadeRuntime.ParticleSystemComponent|SetCustomTimeDilation (Math|Random|RandomFloatinRange 0.68 1.1) _returnvalue)
      (for _index (range 4)
        (Effects|Components|ParticleSystem|SetEmitterEnable _returnvalue (Utilities|String|StringToName (Utilities|String|Append "ShootingStarsSmall" (Utilities|String|ToString(Integer) _index))) (or (<= _index (Math|Float|Round (* (Variables|Default|GetTimeSpanRatio) 3.0))) (Variables|Debug|GetPreviewShootingStars))) ; Turn off additional shooting star emitters based on comet position
      ))))

(event Custom|SpawnMeteor
  ;[1 1 1] Spawn meteor particle system
  (if (or (Variables|Default|GetMeteorsActive) (Variables|Debug|GetPreviewMeteors))
    (if (not (Networking|IsDedicatedServer))
      (bind _returnvalue (Effects|SpawnEmitteratLocation "/Figment_S03/Athena/Events/P_Comet_ShootingStar_Large.P_Comet_ShootingStar_Large" (Utilities|Array|Get(acopy) (Variables|Default|GetShootingStarSpawns) (Variables|Default|GetSpawnLocation)) (Transformation|GetWorldRotation (Variables|Default|GetP_Athena_Event_Comet))))
      (Variables|/Script/CascadeRuntime.ParticleSystemComponent|SetCustomTimeDilation (Math|Random|RandomFloatinRange 0.55 0.9) _returnvalue))))

(fn DefineEventPhase
  (bind _daysuntileventend (Variables|Default|GetDaysUntilEventEnd))
  (Networking|SwitchHasAuthority
    (:Authority
      (Utilities|FlowControl|Sequence
        (:then_0
          (bind (_biseventactive _timesincebegin _timeuntilend _timespanratio) (CalendarEvents|GetCalendarEventInformation (Variables|Default|GetEventName)))
          (bind _timeuntilend_days (Math|Timespan|BreakTimespan _timeuntilend))
          (if _biseventactive
            (Variables|Default|SetEventActive _biseventactive)
            (Variables|Default|SetDaysUntilEventEnd _timeuntilend_days)
            (Variables|Default|SetTimeSpanRatio _timespanratio)
            (else
              _))
          (Development|PrintString (Utilities|String|Append (Variables|Default|GetEventName) "
Days Left: " (Utilities|String|ToString(Integer) _timeuntilend_days) " | Time Span Ratio: " (Utilities|String|ToString(Float) _timespanratio) " | Is Event Active: " (Utilities|String|ToString(Boolean) _biseventactive)) true true "(R=0.000000,G=0.660000,B=1.000000,A=1.000000)" 6e+01))
        (:then_1
          ;[0 1 1 0.5] Set current Meteor Phase based on date
          (if (Variables|Default|GetEventActive)
            (if (<= _daysuntileventend (Variables|Default|GetDaysLeft_Phase4))
              (Variables|Default|SetEventPhase 4)
              (elif (<= _daysuntileventend (Variables|Default|GetDaysLeft_Phase3))
                (Variables|Default|SetEventPhase 3)
                (elif (<= _daysuntileventend (Variables|Default|GetDaysLeft_Phase2))
                  (Variables|Default|SetEventPhase 2)
                  (elif (<= _daysuntileventend (Variables|Default|GetDaysLeft_Phase1))
                    (Variables|Default|SetEventPhase 1)
                    (else
                      (Variables|Default|SetEventPhase 0))))))))
        (:then_2
          ; Debug and Override Setup
          (bind _outxy (DataTable|EvaluateCurveTableRow "/Figment_LootTables/DataTables/FigmentGameData.FigmentGameData" "Debug.EventPhase.Override")
            (:RowFound
              (if (>= _outxy 0)
                (Variables|Default|SetEventActive true)
                (bind _output_get (Variables|Default|SetEventPhase (Math|Float|Truncate _outxy)))
                (bind _output_get_2 (Variables|Default|SetTimeSpanRatio (* 0.25 _output_get)))
                (Development|PrintString (Utilities|String|Append "New Time Span Ratio: " (Utilities|String|ToString(Float) _output_get_2)) true true "(R=0.000000,G=0.660000,B=1.000000,A=1.000000)" 3e+01)
                (else
                  _))
              (Development|PrintString (Utilities|String|Append "Event Phase: " (Utilities|String|ToString(Integer) (Variables|Default|GetEventPhase))) true true "(R=0.000000,G=0.660000,B=1.000000,A=1.000000)" 3e+01))) ; Check if we're overriding Event Phase
        )))))

(fn OnRep_CometPosition
  (CallFunction|SetupComet)
  (Development|PrintString (Utilities|String|Append "Comet Progress: " (Utilities|String|ToString(Float) (Variables|Default|GetCometPosition))) true true "(R=0.000000,G=0.660000,B=1.000000,A=1.000000)" 6e+01))

(fn RandomizeSpawnLocation
  (Variables|Default|SetSpawnLocation (Math|Random|RandomIntegerinRange 0 (- (Utilities|Array|Length (Variables|Default|GetShootingStarSpawns)) 1)))
  (return))

(fn SetupComet
  (bind (_returnvalue_location _returnvalue_rotation) (Math|Transform|BreakTransform (Math|Transform|Lerp(Transform) (Variables|Default|GetCometMidTransform) (Variables|Default|GetCometEndTransform) (Math|Float|MapRangeClamped (Variables|Default|GetCometPosition) 0.0 1.0 0.0 1.0))))
  (bind (_returnvalue_location_2 _returnvalue_rotation_2) (Math|Transform|BreakTransform (Math|Transform|Lerp(Transform) (Variables|Default|GetCometStartTransform) (Variables|Default|GetCometMidTransform) (Math|Float|MapRangeClamped (Variables|Default|GetCometPosition) 0.0 1.0 0.0 0.5))))
  (bind (_returnvalue_location_3 _returnvalue_rotation_3) (Math|Transform|BreakTransform (Math|Transform|Lerp(Transform) (Variables|Default|GetCometStartTransform) (Variables|Default|GetCometMidTransform) (Math|Float|MapRangeClamped (Variables|Default|GetCometPosition) 0.0 1.0 0.5 1.0))))
  (bind (_cometstarttransform_location _cometstarttransform_rotation) (Math|Transform|BreakTransform (Variables|GetCometStartTransform_Location)))
  (bind _p_athena_event_cometdistant (Variables|Default|GetP_Athena_Event_CometDistant))
  (switch int (Variables|Default|GetEventPhase)
    (:1
      ; Phase 1 - Set Comet to Distant and Scale
      (Components|Activation|SetActive _p_athena_event_cometdistant true true)
      (Transformation|SetWorldScale3D _p_athena_event_cometdistant (Math|Conversions|ToVector(Float) (Variables|Default|GetCometPosition))) ; Comet Position is only used to determine scale in this phase
      (Transformation|SetWorldLocationAndRotation _p_athena_event_cometdistant _cometstarttransform_location _cometstarttransform_rotation false false))
    (:2
      ; Phase 2
      (Components|Activation|SetActive (Variables|Default|GetP_Athena_Event_Comet) true true)
      (Transformation|SetWorldTransform (Variables|Default|GetP_Athena_Event_Comet) (Math|Transform|MakeTransform _returnvalue_location_2 _returnvalue_rotation_2 (Math|Conversions|ToVector(Float) (Math|Curves|GetFloatValue (Variables|Default|GetCometScaleCurve_Figment) (Variables|Default|GetTimeSpanRatio)))) false true)
      (Components|Activation|Activate (Variables|Default|GetAmbient_SkyboxLoop_Cue) false))
    (:3
      ; Phase 3
      (Components|Activation|SetActive (Variables|Default|GetP_Athena_Event_Comet) true true)
      (Transformation|SetWorldTransform (Variables|Default|GetP_Athena_Event_Comet) (Math|Transform|MakeTransform _returnvalue_location_3 _returnvalue_rotation_3 (Math|Conversions|ToVector(Float) (Math|Curves|GetFloatValue (Variables|Default|GetCometScaleCurve_Figment) (Variables|Default|GetTimeSpanRatio)))) false true)
      (Variables|Default|SetShootingStarsActive true)
      (Components|Activation|Activate (Variables|Default|GetAmbient_SkyboxLoop_Cue) false))
    (:4
      ; Phase 4
      (Components|Activation|SetActive (Variables|Default|GetP_Athena_Event_Comet) true true)
      (Transformation|SetWorldTransform (Variables|Default|GetP_Athena_Event_Comet) (Math|Transform|MakeTransform _returnvalue_location _returnvalue_rotation (Math|Conversions|ToVector(Float) (Math|Curves|GetFloatValue (Variables|Default|GetCometScaleCurve_Figment) (Variables|Default|GetTimeSpanRatio)))) false true)
      (Variables|Default|SetShootingStarsActive true)
      (Variables|Default|SetMeteorsActive true)
      (Components|Activation|Activate (Variables|Default|GetAmbient_SkyboxLoop_Cue) false))))

(fn ConstructionScript
  (Utilities|FlowControl|Sequence
    (:then_0
      (Utilities|FlowControl|Sequence
        (:then_0
          (bind _outxy (DataTable|EvaluateCurveTableRow "/Figment_LootTables/DataTables/FigmentGameData.FigmentGameData" "Meteor.PhaseDefinition.DaysLeft" 1.0)
            (:RowFound
              (Variables|Default|SetDaysLeft_Phase1 (Math|Float|Truncate _outxy)))))
        (:then_1
          (bind _outxy_2 (DataTable|EvaluateCurveTableRow "/Figment_LootTables/DataTables/FigmentGameData.FigmentGameData" "Meteor.PhaseDefinition.DaysLeft" 2.0)
            (:RowFound
              (Variables|Default|SetDaysLeft_Phase2 (Math|Float|Truncate _outxy_2)))))
        (:then_2
          (bind _outxy_3 (DataTable|EvaluateCurveTableRow "/Figment_LootTables/DataTables/FigmentGameData.FigmentGameData" "Meteor.PhaseDefinition.DaysLeft" 3.0)
            (:RowFound
              (Variables|Default|SetDaysLeft_Phase3 (Math|Float|Truncate _outxy_3)))))
        (:then_3
          (bind _outxy_4 (DataTable|EvaluateCurveTableRow "/Figment_LootTables/DataTables/FigmentGameData.FigmentGameData" "Meteor.PhaseDefinition.DaysLeft" 4.0)
            (:RowFound
              (Variables|Default|SetDaysLeft_Phase4 (Math|Float|Truncate _outxy_4)))))
        (:then_4
          (Development|PrintString (Utilities|String|Append "Days Left Phase 1: " (Utilities|String|ToString(Integer) (Variables|Default|GetDaysLeft_Phase1)) " | Days Left Phase 2: " (Utilities|String|ToString(Integer) (Variables|Default|GetDaysLeft_Phase2)) " | Days Left Phase 3: " (Utilities|String|ToString(Integer) (Variables|Default|GetDaysLeft_Phase3)) " | Days Left Phase 4: " (Utilities|String|ToString(Integer) (Variables|Default|GetDaysLeft_Phase4))) false true "(R=1.000000,G=1.000000,B=0.000000,A=1.000000)" 3e+01))))
    (:then_1
      (bind _outxy_5 (DataTable|EvaluateCurveTableRow "/Figment_LootTables/DataTables/FigmentGameData.FigmentGameData" "Debug.PreviewShootingStars.Enabled")
        (:RowFound
          (Variables|Debug|SetPreviewShootingStars (== _outxy_5 1.0)))))
    (:then_2
      (bind _outxy_6 (DataTable|EvaluateCurveTableRow "/Figment_LootTables/DataTables/FigmentGameData.FigmentGameData" "Debug.PreviewMeteors.Enabled")
        (:RowFound
          (Variables|Debug|SetPreviewMeteors (== _outxy_6 1.0)))))
    (:then_3
      (bind _outxy_7 (DataTable|EvaluateCurveTableRow "/Figment_LootTables/DataTables/FigmentGameData.FigmentGameData" "Debug.PreviewComet.Enabled")
        (:RowFound
          (Variables|Debug|SetPreviewComet (== _outxy_7 1.0)))))))
